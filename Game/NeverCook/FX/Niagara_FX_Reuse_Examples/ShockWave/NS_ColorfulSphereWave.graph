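# Niagara shockwave effect that flashes a single short-lived, expanding energy sphere, stored among reusable FX examples.
# Its one emitter, Energy_Sphere, bursts one camera-offset mesh particle that grows quickly, then eases out.
# Curves animate four dynamic material parameters, while two stacked ScaleColor modules shift HDR red toward yellow and fade it out.
# Front-end modifier functions scale brightness and size curves differently for gameplay and front end, alongside front-end user booleans.
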NiagaraSystem /Game/NeverCook/FX/Niagara_FX_Reuse_Examples/ShockWave/NS_ColorfulSphereWave
dependencies {
 usedDataInterfaces [&NiagaraDataInterfaceCurve, &NiagaraDataInterfaceColorCurve]
 usedDynamicInputs []
 usedModules [&SystemState]
}
emitters {Energy_Sphere {
  emitterUpdateScript [
   &EmitterState()
   &SpawnBurst_Instantaneous(Spawn Count=1, Spawn Probability=&Lerp_Float(A=1, B=0, Alpha=0), Use Spawn Probability=true)
  ]
  particleSpawnScript [
   &InitializeParticle(
    Color=<rgba 0.5 0.5 0.5 0.15>
    Lifetime=0.35
    Mesh Scale Mode=::NewEnumerator3
    Mesh Uniform Scale=2.5
    Position Offset=<0 0 10>
    UsePositionOffset=true
    Write Material Random=true
   ) {module=InitializeParticle001}
   &CameraOffset(Camera Offset Amount=30)
  ]
  particleUpdateScript [
   &ParticleState()
   &DynamicMaterialParameters(
    Index 0 Param 1=&FloatFromCurve(FloatCurve=DataInterface() {propertyValues {Curve {keys [<at 0.2 1>,<at 0.4 0.5>]}}})
    Index 0 Param 2=&FloatFromCurve(FloatCurve=DataInterface() {propertyValues {Curve {keys [
         {arriveTangent=15.2974, time=0.5, value=0.1}
         {arriveTangent=2.1222, interpMode=RCIM_Cubic, leaveTangent=2.1222, tangentMode=RCTM_User, time=1, value=0.3}
        ]}}})
    Index 0 Param 3=&FloatFromCurve(FloatCurve=DataInterface() {propertyValues {Curve {keys [<at 0 -1>,<at 1 0>]}}})
    Index 0 Param 4=&FloatFromCurve(FloatCurve=DataInterface() {propertyValues {Curve {keys [<at 0 0>,<at 1 1>]}}})
    Module.Param0WriteEnabled=true
    Module.Param1WriteEnabled=true
    Module.Param2WriteEnabled=true
    Module.Param3WriteEnabled=true
   )
   &ScaleColor(
    Scale Alpha=&FloatFromCurve(FloatCurve=DataInterface() {propertyValues {Curve {keys [{arriveTangent=-2.709, interpMode=RCIM_Cubic, leaveTangent=-2.709, tangentMode=RCTM_User, time=0.5, value=1}, <at 1 0 RCIM_Cubic>]}}})
    Scale RGB=&MakeVectorFromLinearColorRGB(Color=&ColorFromCurve(ColorCurve=DataInterface() {propertyValues {
        AlphaCurve {keys []}
        BlueCurve {keys [<at 0.3743 0.0537 RCIM_Cubic>,<at 0.9462 0.5577 RCIM_Cubic>]}
        GreenCurve {keys [<at 0.3743 0 RCIM_Cubic>,<at 0.9462 3.6225 RCIM_Cubic>]}
        RedCurve {keys [<at 0.3743 5 RCIM_Cubic>,<at 0.9462 5 RCIM_Cubic>]}
       }}))
    ScaleA=true
    ScaleRGB=true
   )
   &ScaleColor(
    Scale Alpha=&FloatFromCurve(
     FloatCurve=DataInterface() {propertyValues {Curve {keys [{arriveTangent=-1.582, interpMode=RCIM_Cubic, leaveTangent=-1.582, tangentMode=RCTM_User, time=0, value=1}, <at 1 0 RCIM_Cubic>]}}}
     Scale Curve=&FN_FrontEndFloatModifier(Gameplay=1, Front End=1)
    )
    Scale RGB=&VectorFromFloat(Value=&FloatFromCurve(
      FloatCurve=DataInterface() {propertyValues {Curve {keys [<at 0 1 RCIM_Cubic>,<at 0.2 0.5 RCIM_Cubic>]}}}
      Scale Curve=&FN_FrontEndFloatModifier(Gameplay=1, Front End=0.5)
     ))
    ScaleA=true
    ScaleRGB=true
   ) {module=ScaleColor001}
   &ScaleMeshSize(Scale Factor=&VectorFromFloat(Value=&FloatFromCurve(
      FloatCurve=DataInterface() {propertyValues {Curve {keys [
          {arriveTangent=1.4597, interpMode=RCIM_Cubic, leaveTangent=1.4597, tangentMode=RCTM_User, time=0, value=0.2}
          {arriveTangent=0.851, interpMode=RCIM_Cubic, leaveTangent=0.851, tangentMode=RCTM_User, time=0.3205, value=0.6886}
          {arriveTangent=0.1058, interpMode=RCIM_Cubic, leaveTangent=0.1058, tangentMode=RCTM_User, time=1, value=1}
         ]}}}
      Scale Curve=&FN_FrontEndFloatModifier(Gameplay=1.1, Front End=0.7)
     )))
  ]
  rendererClasses [&NiagaraMeshRendererProperties]
  renderers [{rendererClass=&NiagaraMeshRendererProperties, rendererIndex=0}]
 }}
summary {
 systemName=NS_ColorfulSphereWave
 emitters [{emitterName=Energy_Sphere, rendererClasses [&NiagaraMeshRendererProperties]}]
 userVariables [
  User.bIsFrontEnd {type=NiagaraBool}
  User.bIsFrontEndPreview {type=NiagaraBool}
 ]
}
