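# Niagara system for a short impact shockwave effect, built from a single mesh-rendered emitter, Wave02, with no user parameters.
# An instantaneous burst spawns the mesh with random Z rotation and a camera offset.
# Over its sub-second lifetime, curves grow the scale, cool the color from HDR orange to dim red, and fade alpha.
# A curve-driven dynamic material parameter also animates the mesh's material over particle life.
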NiagaraSystem /Game/NeverCook/FX/Niagara_FX_Reuse_Examples/ShockWave/NS_Impact_Wave
dependencies {
 usedDataInterfaces [&NiagaraDataInterfaceCurve, &NiagaraDataInterfaceColorCurve]
 usedDynamicInputs [&MaskFloatBySpawnGroup]
 usedModules [&SystemState]
}
emitters {Wave02 {
  emitterUpdateScript [
   &EmitterState(Life Cycle Mode=::NewEnumerator1, Loop Behavior=::NewEnumerator1, Loop Duration=1, UseLoopDelay=false)
   &SpawnBurst_Instantaneous(Use Spawn Probability=false, UseLoopCountLimit=false)
  ]
  particleSpawnScript [
   &InitializeParticle(Lifetime=0.8, Mesh Scale=<1.8 1.8 1.8>, Mesh Scale Mode=::NewEnumerator4, UsePositionOffset=false, Write Material Random=true)
   &InitialMeshOrientation(
    Mesh Orientation Mode=::NewEnumerator2
    Rotation=&RandomRangeVector(Minimum=<0 0 0>, Maximum=<0 0 1>)
    Use Rotation Vector=true
   )
   &CameraOffset(Camera Offset Amount=40)
  ]
  particleUpdateScript [
   &ParticleState()
   &ScaleMeshSize(Scale Factor=&Multiply_VectorByFloat(
     Vector=<1 1 1>
     Float=&FloatFromCurve(
      FloatCurve=DataInterface() {propertyValues {Curve {keys [<at 0 0.5>, {arriveTangent=1, interpMode=RCIM_Cubic, leaveTangent=1, time=0.4357, value=1.1912}, <at 1 1.5>]}}}
      Scale Curve=&Multiply_Float(A=1, B=1)
     )
    ))
   &ScaleColor(
    Scale Alpha=&FloatFromCurve(
     FloatCurve=DataInterface() {propertyValues {Curve {keys [
         {arriveTangent=-1.2068, interpMode=RCIM_Cubic, leaveTangent=-1.2068, tangentMode=RCTM_User, time=0, value=1}
         {arriveTangent=-1, interpMode=RCIM_Cubic, leaveTangent=-1, time=0.3, value=0.925}
         {tangentMode=RCTM_User, time=1, value=0}
        ]}}}
     Scale Curve=2
    )
    Scale RGB=&MakeVectorFromLinearColorRGB(Color=&ColorFromCurve(ColorCurve=DataInterface() {propertyValues {
        AlphaCurve {keys []}
        BlueCurve {keys [<at 0 0 RCIM_Cubic>,<at 0.4508 0.2083>,<at 0.5798 0.0156>]}
        GreenCurve {keys [<at 0 2.133 RCIM_Cubic>,<at 0.4508 0.5414>,<at 0.5798 0.0332>]}
        RedCurve {keys [<at 0 20 RCIM_Cubic>,<at 0.4508 2>,<at 0.5798 0.5>]}
       }}))
    ScaleA=true
    ScaleRGB=true
   )
   &DynamicMaterialParameters(
    Index 0 Param 1=&FloatFromCurve(FloatCurve=DataInterface() {propertyValues {Curve {keys [<at 0 -0.5>,<at 1 -0.35>]}}})
    Module.Index0Param0Write=true
    Module.Index0Param1Write=true
    Module.Index0Param2Write=true
    Module.Index0Param3Write=true
   )
  ]
  rendererClasses [&NiagaraMeshRendererProperties]
  renderers [{rendererClass=&NiagaraMeshRendererProperties, rendererIndex=0}]
 }}
summary {
 systemName=NS_Impact_Wave
 emitters [{emitterName=Wave02, rendererClasses [&NiagaraMeshRendererProperties]}]
 userVariables []
}
