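# NS_Spine_ShockWave is a Niagara shockwave effect with one emitter, FastWave001, that bursts two short-lived, oriented mesh particles.
# Curves expand each mesh quickly in Y and Z, then more slowly, while X scale stays fixed, so the wave widens without thickening.
# Color curves shift the HDR base color from black to reddish as alpha fades out, while a rising curve animates a dynamic material parameter.
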NiagaraSystem /Game/NeverCook/FX/Niagara_FX_Reuse_Examples/ShockWave/NS_Spine_ShockWave
dependencies {
 usedDataInterfaces [&NiagaraDataInterfaceCurve, &NiagaraDataInterfaceColorCurve]
 usedDynamicInputs []
 usedModules [&SystemState]
}
emitters {FastWave001 {
  emitterUpdateScript [
   &EmitterState(Life Cycle Mode=::NewEnumerator1, Loop Behavior=::NewEnumerator1, Loop Duration=1, UseLoopDelay=false)
   &SpawnBurst_Instantaneous(Spawn Count=2, Use Spawn Probability=false)
  ]
  particleSpawnScript [
   &InitializeParticle(
    Color=<rgba 5 5 5 1>
    Lifetime=0.5
    Mesh Scale=<1 2 2>
    Mesh Scale Mode=::NewEnumerator4
    UsePositionOffset=false
    Write Material Random=true
   )
   &InitialMeshOrientation(Rotation=<0 0.25 0>, Use Orientation Vector=true, Use Rotation Vector=true) {module=InitialMeshOrientation002}
  ]
  particleUpdateScript [
   &ParticleState()
   &ScaleMeshSize(Scale Factor=&MakeVector(
     X=1
     Y=&FloatFromCurve(FloatCurve=DataInterface() {propertyValues {Curve {keys [
          {arriveTangent=3.6192, interpMode=RCIM_Cubic, leaveTangent=3.6192, tangentMode=RCTM_User, time=0, value=1}
          {arriveTangent=1, interpMode=RCIM_Cubic, leaveTangent=1, time=0.2, value=1.7}
          <at 1 2 RCIM_Cubic>
         ]}}})
     Z=&FloatFromCurve(FloatCurve=DataInterface() {propertyValues {Curve {keys [
          {arriveTangent=3.6192, interpMode=RCIM_Cubic, leaveTangent=3.6192, tangentMode=RCTM_User, time=0, value=1}
          {arriveTangent=1, interpMode=RCIM_Cubic, leaveTangent=1, time=0.2, value=1.7}
          <at 1 2 RCIM_Cubic>
         ]}}})
    )) {module=ScaleMeshSize001}
   &ScaleColor(
    Scale Alpha=&FloatFromCurve(FloatCurve=DataInterface() {propertyValues {Curve {keys [{arriveTangent=-1.6382, interpMode=RCIM_Cubic, leaveTangent=-1.6382, tangentMode=RCTM_User, time=0, value=1}, <at 1 0 RCIM_Cubic>]}}})
    Scale RGB=&MakeVectorFromLinearColorRGB(Color=&ColorFromCurve(ColorCurve=DataInterface() {propertyValues {
        AlphaCurve {keys []}
        BlueCurve {keys [{arriveTangent=2.3578, interpMode=RCIM_Cubic, time=0.2851, value=0}, <at 0.3594 0 RCIM_Cubic>]}
        GreenCurve {keys [{arriveTangent=-30.4531, interpMode=RCIM_Cubic, time=0.2851, value=0}, <at 0.3594 0.0457 RCIM_Cubic>]}
        RedCurve {keys [{arriveTangent=-115.5308, interpMode=RCIM_Cubic, time=0.2851, value=0}, <at 0.3594 0.3 RCIM_Cubic>]}
       }}))
    ScaleA=true
    ScaleRGB=true
   ) {module=ScaleColor002}
   &DynamicMaterialParameters(
    Index 0 Param 1=&FloatFromCurve(FloatCurve=DataInterface() {propertyValues {Curve {keys [
         {arriveTangent=2.2752, interpMode=RCIM_Cubic, leaveTangent=2.2752, tangentMode=RCTM_User, time=0, value=-0.4}
         {arriveTangent=0.4116, interpMode=RCIM_Cubic, leaveTangent=0.4116, tangentMode=RCTM_User, time=1, value=0.925}
        ]}}})
    Module.Param0WriteEnabled=true
    Module.Param1WriteEnabled=false
    Module.Param2WriteEnabled=false
    Module.Param3WriteEnabled=false
   )
  ]
  rendererClasses [&NiagaraMeshRendererProperties]
  renderers [{rendererClass=&NiagaraMeshRendererProperties, rendererIndex=0}]
 }}
summary {
 systemName=NS_Spine_ShockWave
 emitters [{emitterName=FastWave001, rendererClasses [&NiagaraMeshRendererProperties]}]
 userVariables []
}
