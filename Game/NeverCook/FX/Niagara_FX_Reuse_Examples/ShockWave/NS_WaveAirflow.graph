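# Niagara shockwave effect, from a folder of reusable FX examples, that shows a short expanding, fading wave of glowing orange energy.
# A single Ground_energy emitter uses an instantaneous burst to spawn short-lived mesh particles with a slightly flattened scale and random material value.
# Over each particle's life, curves grow the mesh quickly then ease off, fade its alpha, and push HDR color to a bright mid-life flare before darkening.
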NiagaraSystem /Game/NeverCook/FX/Niagara_FX_Reuse_Examples/ShockWave/NS_WaveAirflow
dependencies {
 usedDataInterfaces [&NiagaraDataInterfaceCurve, &NiagaraDataInterfaceColorCurve]
 usedDynamicInputs []
 usedModules [&SystemState]
}
emitters {Ground_energy {
  emitterUpdateScript [
   &EmitterState(Life Cycle Mode=::NewEnumerator1, Loop Behavior=::NewEnumerator1, Loop Duration=1, UseLoopDelay=false)
   &SpawnBurst_Instantaneous(Use Spawn Probability=false, UseLoopCountLimit=false)
  ]
  particleSpawnScript [&InitializeParticle(
    Color=<rgba 2 2 2 1>
    Lifetime=0.65
    Mesh Scale=<3 3 2>
    Mesh Scale Mode=::NewEnumerator4
    UsePositionOffset=true
    Write Material Random=true
   )]
  particleUpdateScript [
   &ParticleState()
   &ScaleMeshSize(Scale Factor=&Multiply_VectorByFloat(
     Vector=<0.8 0.8 0.8>
     Float=&FloatFromCurve(FloatCurve=DataInterface() {propertyValues {Curve {keys [
          {arriveTangent=4.1785, interpMode=RCIM_Cubic, leaveTangent=4.1785, tangentMode=RCTM_User, time=0, value=0}
          {arriveTangent=1.1179, interpMode=RCIM_Cubic, leaveTangent=1.1179, tangentMode=RCTM_User, time=0.2, value=0.9}
          {arriveTangent=0.0224, interpMode=RCIM_Cubic, leaveTangent=0.0224, tangentMode=RCTM_User, time=0.8, value=1.5}
         ]}}})
    ))
   &ScaleColor(
    Scale Alpha=&Multiply_Float(
     A=1
     B=&FloatFromCurve(FloatCurve=DataInterface() {propertyValues {Curve {keys [
          {arriveTangent=-2.5289, interpMode=RCIM_Cubic, leaveTangent=-2.5289, tangentMode=RCTM_User, time=0, value=1}
          {arriveTangent=-0.9759, interpMode=RCIM_Cubic, leaveTangent=-0.9759, tangentMode=RCTM_User, time=0.5, value=0.55}
          {arriveTangent=-1.3542, interpMode=RCIM_Cubic, leaveTangent=-1.3542, tangentMode=RCTM_User, time=0.8, value=0}
         ]}}})
    )
    Scale RGB=&MakeVectorFromLinearColorRGB(Color=&ColorFromCurve(ColorCurve=DataInterface() {propertyValues {
        AlphaCurve {keys []}
        BlueCurve {keys [<at 0 0 RCIM_Cubic>,<at 0.4206 0.5734>,<at 0.4755 0.4592>,<at 1 0 RCIM_Cubic>]}
        GreenCurve {keys [<at 0 0.6694 RCIM_Cubic>,<at 0.4206 1.2584>,<at 0.4755 2.1601>,<at 1 0 RCIM_Cubic>]}
        RedCurve {keys [<at 0 2 RCIM_Cubic>,<at 0.4206 2>,<at 0.4755 10>,<at 1 0 RCIM_Cubic>]}
       }}))
    ScaleA=true
    ScaleRGB=true
   )
  ]
  rendererClasses [&NiagaraMeshRendererProperties]
  renderers [{rendererClass=&NiagaraMeshRendererProperties, rendererIndex=0}]
 }}
summary {
 systemName=NS_WaveAirflow
 emitters [{emitterName=Ground_energy, rendererClasses [&NiagaraMeshRendererProperties]}]
 userVariables []
}
